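# Physics asset for the Lena female cosmetic's FaceAcc_Bacpack mesh part, covering the head and one side-hair bone chain.
# Three capsule bodies are used: a large one on the head and small ones on the base and dynamic hairSide bones.
# The dynamic hair bone has tight swing and near-locked twist limits against its base.
# The base-to-head constraint leaves all angular motion free, so its listed degree limits do not restrict rotation.
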PhysicsAsset /BRCosmetics/Characters/Player/Female/Medium/Bodies/F_MED_Lena/Meshes/Parts/F_MED_Lena_FaceAcc_Bacpack_Physics
bodies {
 base_hairSide_00 [{shapeType=Capsule, length=0.505, radius=0.505, shapeName=base_hairSide_00_capsule, rotation=<rot 0 -90 0>}]
 dyn_hairSide_00 [{shapeType=Capsule, length=2, radius=1, shapeName=dyn_hairSide_00_capsule, center=<2 0 0>, rotation=<rot 90 0 0>}]
 head [{shapeType=Capsule, length=3.75, radius=8.5, shapeName=head_capsule, center=<-0.202 3.5179 0>, rotation=<rot 90 180 180>}]
}
constraints [
 {bone1Name=dyn_hairSide_00, bone2Name=base_hairSide_00, swing1LimitDegrees=20, swing2LimitDegrees=25, twistLimitDegrees=1}
 {
  bone1Name=base_hairSide_00
  bone2Name=head
  swing1LimitDegrees=45
  swing1Motion=Free
  swing2LimitDegrees=45
  swing2Motion=Free
  twistLimitDegrees=45
  twistMotion=Free
 }
]
