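# Physics asset for the F_MED_Lena face accessory mesh, covering a braid tail and a side hair strand.
# Capsule bodies form a four-segment dynamic braid chain and a single dynamic hair segment, each hung from a small base body linked to the head.
# These dynamic joints have limited swing and very tight twist limits.
# Head, neck, spine and upper-arm capsules supply body volumes, joined by fully free angular constraints.
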PhysicsAsset /BRCosmetics/Characters/Player/Female/Medium/Bodies/F_MED_Lena/Meshes/Parts/F_MED_Lena_FaceAcc_Physics
bodies {
 base_braidTail_00 [{shapeType=Capsule, length=0.505, radius=0.505, shapeName=base_braidTail_00_capsule, rotation=<rot 0 -90 0>}]
 base_hairSide_00 [{shapeType=Capsule, length=0.505, radius=0.505, shapeName=base_hairSide_00_capsule, rotation=<rot 0 -90 0>}]
 dyn_braidTail_00 [{shapeType=Capsule, length=7.7967, radius=1, shapeName=dyn_braidTail_00_capsule, center=<2.4985 0.5275 0>, rotation=<rot 90 0 0>}]
 dyn_braidTail_01 [{shapeType=Capsule, length=4.9246, radius=3.2243, shapeName=dyn_braidTail_01_capsule, center=<4.8793 -1.6402 0.2883>, rotation=<rot 90 0 0>}]
 dyn_braidTail_02 [{shapeType=Capsule, length=4.5065, radius=2.786, shapeName=dyn_braidTail_02_capsule, center=<4.6447 -1.375 0.2486>, rotation=<rot 90 0 0>}]
 dyn_braidTail_03 [{
   shapeType=Capsule
   length=4.3765
   radius=2.7097
   shapeName=dyn_braidTail_03_capsule
   center=<5.7643 -0.6089 0.1163>
   rotation=<rot 90 -178.8065 171.8198>
  }]
 dyn_hairSide_00 [{shapeType=Capsule, length=2, radius=1, shapeName=dyn_hairSide_00_capsule, center=<2 0 0>, rotation=<rot 90 0 0>}]
 head [{shapeType=Capsule, length=4.4205, radius=7.8295, shapeName=head_capsule, center=<-0.202 0.9408 0>, rotation=<rot 90 0 0>}]
 neck_01 [{shapeType=Capsule, length=20.3328, radius=5, shapeName=neck_01_capsule, center=<0.8327 1.2585 0>}]
 neck_02 [{shapeType=Capsule, length=19.2751, radius=4, shapeName=neck_02_capsule, center=<0 0.4552 0>}]
 spine_03 [{shapeType=Capsule, length=25.505, radius=10.5, shapeName=spine_03_capsule, center=<9.5256 1.9385 0>, rotation=<rot 0 -90 0>}]
 spine_05 [{shapeType=Capsule, length=16.755, radius=9.255, shapeName=spine_05_capsule, center=<2.5115 0.8813 0>, rotation=<rot 0 -90 0>}]
 upperarm_l [{shapeType=Capsule, length=17.3628, radius=5.448, shapeName=upperarm_l_capsule, center=<8.9717 0 0>, rotation=<rot 90 0 0>}]
 upperarm_r [{shapeType=Capsule, length=17.3628, radius=5.448, shapeName=upperarm_r_capsule, center=<-8.9717 0 0>, rotation=<rot 90 0 0>}]
}
constraints [
 {bone1Name=dyn_braidTail_00, bone2Name=base_braidTail_00, swing1LimitDegrees=35, swing2LimitDegrees=35, twistLimitDegrees=1}
 {bone1Name=dyn_braidTail_01, bone2Name=dyn_braidTail_00, swing1LimitDegrees=30, swing2LimitDegrees=30, twistLimitDegrees=1}
 {bone1Name=dyn_braidTail_02, bone2Name=dyn_braidTail_01, swing1LimitDegrees=30, swing2LimitDegrees=30, twistLimitDegrees=1}
 {bone1Name=dyn_braidTail_03, bone2Name=dyn_braidTail_02, swing1LimitDegrees=35, swing2LimitDegrees=35, twistLimitDegrees=1}
 {
  bone1Name=base_hairSide_00
  bone2Name=head
  swing1LimitDegrees=45
  swing1Motion=Free
  swing2LimitDegrees=45
  swing2Motion=Free
  twistLimitDegrees=45
  twistMotion=Free
 }
 {
  bone1Name=base_braidTail_00
  bone2Name=head
  swing1LimitDegrees=45
  swing1Motion=Free
  swing2LimitDegrees=45
  swing2Motion=Free
  twistLimitDegrees=45
  twistMotion=Free
 }
 {
  bone1Name=head
  bone2Name=neck_02
  swing1LimitDegrees=45
  swing1Motion=Free
  swing2LimitDegrees=45
  swing2Motion=Free
  twistLimitDegrees=45
  twistMotion=Free
 }
 {
  bone1Name=neck_02
  bone2Name=neck_01
  swing1LimitDegrees=45
  swing1Motion=Free
  swing2LimitDegrees=45
  swing2Motion=Free
  twistLimitDegrees=45
  twistMotion=Free
 }
 {
  bone1Name=neck_01
  bone2Name=spine_05
  swing1LimitDegrees=45
  swing1Motion=Free
  swing2LimitDegrees=45
  swing2Motion=Free
  twistLimitDegrees=45
  twistMotion=Free
 }
 {
  bone1Name=spine_05
  bone2Name=spine_03
  swing1LimitDegrees=45
  swing1Motion=Free
  swing2LimitDegrees=45
  swing2Motion=Free
  twistLimitDegrees=45
  twistMotion=Free
 }
 {
  bone1Name=upperarm_l
  bone2Name=spine_05
  swing1LimitDegrees=45
  swing1Motion=Free
  swing2LimitDegrees=45
  swing2Motion=Free
  twistLimitDegrees=45
  twistMotion=Free
 }
 {
  bone1Name=upperarm_r
  bone2Name=spine_05
  swing1LimitDegrees=45
  swing1Motion=Free
  swing2LimitDegrees=45
  swing2Motion=Free
  twistLimitDegrees=45
  twistMotion=Free
 }
 {bone1Name=dyn_hairSide_00, bone2Name=base_hairSide_00, swing1LimitDegrees=25, swing2LimitDegrees=40, twistLimitDegrees=1}
]
