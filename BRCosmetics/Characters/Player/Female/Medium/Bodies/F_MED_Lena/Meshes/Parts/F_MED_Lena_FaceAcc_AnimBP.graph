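# Animation Blueprint for Lena's face accessory part, animating her braid and hat on top of the pose copied from the attached parent mesh.
# Without back bling, it blends between head-anchored constraint posing and spine-based RigidBody physics, depending on whether rigid bodies are enabled.
# Back bling or capes select a constrained braid chain with its own backpack physics.
# The event graph derives gravity, head tilt, skydiving and crouch-aim values driving corrective bone offsets.
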AnimBlueprint /BRCosmetics/Characters/Player/Female/Medium/Bodies/F_MED_Lena/Meshes/Parts/F_MED_Lena_FaceAcc_AnimBP
target_skeleton=&/F_MED_Lena_FaceAcc_Skeleton
graphs [
 AnimGraph {nodes [
   Local To Component()
   n1 = Use cached pose 'Normal Dynamics'()
   n2 = Use cached pose 'Backpack Dynamics'()
   n3 = Get is Wearing Backbling() -> (is Wearing Backbling:bool)
   Output Pose(Result=Blend Poses by bool(
     BlendPose_0=Use cached pose 'Backpack Dynamics'()
     BlendPose_1=Blend Poses by bool(BlendPose_0=@n2, BlendPose_1=@n1, bActiveValue=@n3) {blendType=Linear}
     bActiveValue=Get is Wearing Cape() -> (is Wearing Cape:bool)
    ) {blendType=Linear}) {layerGroup=None}
   Save cached pose 'Start Pose'(Pose=Copy Pose From Mesh(SourceMeshComponent=Get Skel Mesh()) {bUseAttachedParent=true}) {cacheName=Start Pose}
   n4 = Get head_fwdBwd_rot() -> (head_fwdBwd_rot:double)
   n5 = "Transform (Modify) Bone - Bone: dyn_braidTail_03"(
    ComponentPose=Local To Component(LocalPose=Use cached pose 'Start Pose'())
    Alpha=Get ForwardCrouchAiming() -> (ForwardCrouchAiming:double)
    Translation="0,0,0"
    Rotation="0,0,0"
    Scale="1,1,1"
   ) {
    rotationMode=BMM_Additive
    rotationSpace=BCS_BoneSpace
    translationMode=BMM_Replace
    translationSpace=BCS_BoneSpace
    alphaBoolBlend {blendOption=Linear}
    alphaScaleBiasClamp {bInterpResult=true}
    boneToModify {boneName=dyn_braidTail_03}
   }
   n6 = "Constraint - dyn_braidTail_00 (PP)"!(ComponentPose=@n5, Alpha=1, ConstraintWeights_0=0.85, ConstraintWeights_1:float=0.15) -> (Pose) {
    alphaBoolBlend {blendOption=Linear}
    boneToModify {boneName=dyn_braidTail_00}
    constraintSetup [{targetBone {boneName=head}}, {targetBone {boneName=spine_05}}]
    constraintWeights [1,1]
   }
   n7 = "Constraint - dyn_braidTail_01 (PP)"!(ComponentPose=@n6, Alpha=1, ConstraintWeights_0=0.05, ConstraintWeights_1:float=0.95) -> (Pose) {
    alphaBoolBlend {blendOption=Linear}
    boneToModify {boneName=dyn_braidTail_01}
    constraintSetup [{targetBone {boneName=head}}, {targetBone {boneName=spine_05}}]
    constraintWeights [1,1]
   }
   n8 = "Constraint - dyn_braidTail_02 (PP)"!(ComponentPose=@n7, Alpha=1, ConstraintWeights_0=0.3, ConstraintWeights_1:float=0.7) -> (Pose) {
    alphaBoolBlend {blendOption=Linear}
    boneToModify {boneName=dyn_braidTail_02}
    constraintSetup [{targetBone {boneName=spine_04}}, {targetBone {boneName=spine_05}}]
    constraintWeights [1,1]
   }
   n9 = "Constraint - dyn_braidTail_03 (PP)"!(ComponentPose=@n8, Alpha=1, ConstraintWeights_0=0.5, ConstraintWeights_1:float=0.5) -> (Pose) {
    alphaBoolBlend {blendOption=Linear}
    boneToModify {boneName=dyn_braidTail_03}
    constraintSetup [{targetBone {boneName=spine_05}}, {targetBone {boneName=spine_04}}]
    constraintWeights [1,1]
   }
   n10 = "Transform (Modify) Bone - Bone: dyn_braidTail_00"(ComponentPose=@n9, Alpha=@n4, Translation="-0.858,-1.724,-0.765", Rotation="-0.204,6.101,-0.012", Scale="1,1,1") {
    rotationMode=BMM_Additive
    rotationSpace=BCS_BoneSpace
    translationMode=BMM_Additive
    translationSpace=BCS_BoneSpace
    alphaBoolBlend {blendOption=Linear}
    alphaScaleBiasClamp {bMapRange=true, inRange {max=0, min=-25}, outRange {max=0, min=1}}
    boneToModify {boneName=dyn_braidTail_00}
    rotation=<rot -0.204 6.101 -0.012>
    translation=<-0.858 -1.724 -0.765>
   }
   n11 = "Transform (Modify) Bone - Bone: dyn_braidTail_01"(ComponentPose=@n10, Alpha=@n4, Translation="-0.001,-0.467,-0.240", Rotation="-0.058,4.677,0.109", Scale="1,1,1") {
    rotationMode=BMM_Additive
    rotationSpace=BCS_BoneSpace
    translationMode=BMM_Additive
    translationSpace=BCS_BoneSpace
    alphaBoolBlend {blendOption=Linear}
    alphaScaleBiasClamp {bMapRange=true, inRange {max=0, min=-25}, outRange {max=0, min=1}}
    boneToModify {boneName=dyn_braidTail_01}
    rotation=<rot -0.058 4.6771 0.109>
    translation=<-0.001 -0.467 -0.24>
   }
   n12 = "Transform (Modify) Bone - Bone: dyn_braidTail_03"(ComponentPose=@n11, Alpha=1, Translation="-0.278,1.289,-0.098", Rotation="-0.010,25.811,0.030", Scale="1,0.5,1") {
    rotationMode=BMM_Additive
    rotationSpace=BCS_BoneSpace
    scaleMode=BMM_Additive
    scaleSpace=BCS_BoneSpace
    translationMode=BMM_Additive
    translationSpace=BCS_BoneSpace
    alphaBoolBlend {blendOption=Linear}
    boneToModify {boneName=dyn_braidTail_03}
    rotation=<rot -0.01 25.811 0.03>
    scale {y=0.5}
    translation=<-0.278 1.289 -0.098>
   }
   n13 = Get ForwardCrouchAiming()
   n14 = "Transform (Modify) Bone - Bone: hat"(ComponentPose=@n12, Alpha=@n13, Translation="0,0,0", Rotation="0,0,-1", Scale="1,1,1") {
    rotationMode=BMM_Additive
    rotationSpace=BCS_BoneSpace
    translationMode=BMM_Additive
    translationSpace=BCS_BoneSpace
    alphaBoolBlend {blendOption=Linear}
    alphaScaleBiasClamp {bInterpResult=true}
    boneToModify {boneName=hat}
    rotation {roll=-1}
   }
   n15 = "Transform (Modify) Bone - Bone: hatBrim"(ComponentPose=@n14, Alpha=@n13, Translation="0,0,0", Rotation="0,0,-21.968", Scale="1,1,1") {
    rotationMode=BMM_Additive
    rotationSpace=BCS_BoneSpace
    translationMode=BMM_Additive
    translationSpace=BCS_BoneSpace
    alphaBoolBlend {blendOption=Linear}
    alphaScaleBiasClamp {bInterpResult=true}
    boneToModify {boneName=hatBrim}
    rotation {roll=-21.968}
   }
   n16 = RigidBody!(ComponentPose=@n15, Alpha, OverrideWorldGravity:Vector=Get GravityOverride(), ExternalForce) -> (Pose) {
    bOverrideWorldGravity=true
    overridePhysicsAsset=&/F_MED_Lena_FaceAcc_Bacpack_Physics
    simulationSpace=BaseBoneSpace
    alphaBoolBlend {blendOption=Linear}
    baseBoneRef {boneName=head}
    componentAppliedLinearAccClamp=<3000 2500 3000>
    componentLinearAccScale=<0.3 0.25 0.3>
    componentLinearVelScale=<0.3 0.25 0.3>
    simSpaceSettings {
     maxAngularAcceleration=1000
     maxAngularVelocity=3
     maxLinearAcceleration=700
     maxLinearVelocity=500
     velocityScaleZ=0
     worldAlpha=0.2
     externalLinearDragV=<0.2 0.2>
    }
   }
   Save cached pose 'Backpack Dynamics'(Pose=Component To Local(ComponentPose=@n16)) {cacheName=Backpack Dynamics}
   n17 = Get head_fwdBwd_rot()
   n18 = "Constraint - dyn_braidTail_00 (PP)"(ComponentPose=Local To Component(LocalPose=Use cached pose 'Start Pose'()), Alpha=1, ConstraintWeights_0=0.85, ConstraintWeights_1=0.15) {
    alphaBoolBlend {blendOption=Linear}
    boneToModify {boneName=dyn_braidTail_00}
    constraintSetup [{targetBone {boneName=head}}, {targetBone {boneName=spine_05}}]
    constraintWeights [1,1]
   }
   n19 = "Constraint - dyn_braidTail_01 (PP)"(ComponentPose=@n18, Alpha=1, ConstraintWeights_0=0.05, ConstraintWeights_1=0.95) {
    alphaBoolBlend {blendOption=Linear}
    boneToModify {boneName=dyn_braidTail_01}
    constraintSetup [{targetBone {boneName=head}}, {targetBone {boneName=spine_05}}]
    constraintWeights [1,1]
   }
   n20 = "Constraint - dyn_braidTail_02 (PP)"(ComponentPose=@n19, Alpha=1, ConstraintWeights_0=0, ConstraintWeights_1=1) {
    alphaBoolBlend {blendOption=Linear}
    boneToModify {boneName=dyn_braidTail_02}
    constraintSetup [{targetBone {boneName=head}}, {targetBone {boneName=spine_05}}]
    constraintWeights [1,1]
   }
   n21 = "Constraint - dyn_braidTail_03 (P)"(ComponentPose=@n20, Alpha=1, ConstraintWeights_0=1) {
    alphaBoolBlend {blendOption=Linear}
    boneToModify {boneName=dyn_braidTail_03}
    constraintSetup [{targetBone {boneName=spine_05}}]
    constraintWeights [1]
   }
   n22 = "Transform (Modify) Bone - Bone: dyn_braidTail_00"(ComponentPose=@n21, Alpha=@n17, Translation="0.015,-5,0", Rotation="0.002,-5,0.001", Scale="1,1,1") {
    rotationMode=BMM_Additive
    rotationSpace=BCS_BoneSpace
    translationMode=BMM_Additive
    translationSpace=BCS_BoneSpace
    alphaBoolBlend {blendOption=Linear}
    alphaScaleBiasClamp {bMapRange=true, inRange {max=0, min=-25}, outRange {max=0, min=1}}
    boneToModify {boneName=dyn_braidTail_00}
    rotation=<rot 0.002 -5 0.001>
    translation=<0.015 -5>
   }
   n23 = "Transform (Modify) Bone - Bone: dyn_braidTail_01"(ComponentPose=@n22, Alpha=@n17, Translation="-0.001,-0.467,-0.240", Rotation="-0.010,-2.865,0.003", Scale="1,1,1") {
    rotationMode=BMM_Additive
    rotationSpace=BCS_BoneSpace
    translationMode=BMM_Additive
    translationSpace=BCS_BoneSpace
    alphaBoolBlend {blendOption=Linear}
    alphaScaleBiasClamp {bMapRange=true, inRange {max=0, min=-25}, outRange {max=0, min=1}}
    boneToModify {boneName=dyn_braidTail_01}
    rotation=<rot -0.01 -2.865 0.003>
    translation=<-0.001 -0.467 -0.24>
   }
   n24 = Get head_fwdBwd_rot()
   n25 = "Transform (Modify) Bone - Bone: dyn_braidTail_00"(ComponentPose=@n23, Alpha=@n24, Translation="0,1,0", Rotation="0,0,0", Scale="1,1,1") {
    rotationMode=BMM_Additive
    rotationSpace=BCS_BoneSpace
    translationMode=BMM_Additive
    translationSpace=BCS_BoneSpace
    alphaBoolBlend {blendOption=Linear}
    alphaScaleBiasClamp {bMapRange=true, inRange {max=15}}
    boneToModify {boneName=dyn_braidTail_00}
    translation {y=1}
   }
   n26 = "Transform (Modify) Bone - Bone: dyn_braidTail_01"(ComponentPose=@n25, Alpha=@n24, Translation="0.009,1,0", Rotation="0,0,0", Scale="1,1,1") {
    rotationMode=BMM_Additive
    rotationSpace=BCS_BoneSpace
    translationMode=BMM_Additive
    translationSpace=BCS_BoneSpace
    alphaBoolBlend {blendOption=Linear}
    alphaScaleBiasClamp {bMapRange=true, inRange {max=15}}
    boneToModify {boneName=dyn_braidTail_01}
    translation=<0.009 1>
   }
   n27 = "Transform (Modify) Bone - Bone: dyn_braidTail_02"(ComponentPose=@n26, Alpha=@n24, Translation="0.009,1,0", Rotation="-0.009,11.852,0.002", Scale="1,1,1") {
    rotationMode=BMM_Additive
    rotationSpace=BCS_BoneSpace
    translationMode=BMM_Additive
    translationSpace=BCS_BoneSpace
    alphaBoolBlend {blendOption=Linear}
    alphaScaleBiasClamp {bMapRange=true, inRange {max=15}}
    boneToModify {boneName=dyn_braidTail_02}
    rotation=<rot -0.009 11.8521 0.002>
    translation=<0.009 1>
   }
   n28 = "Transform (Modify) Bone - Bone: base_braidTail_00"!(ComponentPose=Local To Component(LocalPose=Use cached pose 'Start Pose'()), Alpha=Get Hair Up() -> (Hair Up:double), Rotation="0,-15,0") -> (Pose) {
    rotationMode=BMM_Additive
    rotationSpace=BCS_BoneSpace
    alphaBoolBlend {blendOption=Linear}
    boneToModify {boneName=base_braidTail_00}
    rotation {yaw=-15}
   }
   n29 = RigidBody(ComponentPose=@n28, OverrideWorldGravity=Get GravityOverride()) {
    bClampLinearTranslationLimitToRefPose=true
    bOverrideWorldGravity=true
    overridePhysicsAsset=&/F_MED_Lena_FaceAcc_Physics
    simulationSpace=BaseBoneSpace
    alphaBoolBlend {blendOption=Linear}
    baseBoneRef {boneName=spine_05}
    componentAppliedLinearAccClamp=<3000 3000 3000>
    componentLinearAccScale=<0.02 0.02 0.05>
    componentLinearVelScale=<0.02 0.02 0.05>
    simSpaceSettings {velocityScaleZ=0.73, worldAlpha=0.2}
   }
   n30 = "Transform (Modify) Bone - Bone: hatBrim"(ComponentPose=@n29, Alpha=Get is Skydiving() -> (is Skydiving:double), Translation="0,0,0", Rotation="0,0,3.2", Scale="1,1,1") {
    rotationMode=BMM_Additive
    rotationSpace=BCS_BoneSpace
    alphaBoolBlend {blendOption=Linear}
    boneToModify {boneName=hatBrim}
    rotation {roll=3.2}
   }
   n31 = Get head_fwdBwd_rot()
   n32 = "Transform (Modify) Bone - Bone: hat"(ComponentPose=@n30, Alpha=@n31, Translation="0,0,0", Rotation="0,0,5", Scale="1,1,1") {
    rotationMode=BMM_Additive
    rotationSpace=BCS_BoneSpace
    alphaBoolBlend {blendOption=Linear}
    alphaScaleBiasClamp {bInterpResult=true}
    boneToModify {boneName=hat}
    rotation {roll=5}
   }
   n33 = "Transform (Modify) Bone - Bone: hatBrim"(ComponentPose=@n32, Alpha=@n31, Translation="0,0,0", Rotation="0,0,15", Scale="1,1,1") {
    rotationMode=BMM_Additive
    rotationSpace=BCS_BoneSpace
    alphaBoolBlend {blendOption=Linear}
    alphaScaleBiasClamp {bInterpResult=true}
    boneToModify {boneName=hatBrim}
    rotation {roll=15}
   }
   Save cached pose 'Normal Dynamics'(Pose=Blend Poses by bool(BlendPose_0=Component To Local(ComponentPose=@n33), BlendPose_1=Component To Local(ComponentPose=@n27), bActiveValue=Get bRigidBodyEnabled()) {blendType=Linear}) {cacheName=Normal Dynamics}
   Reroute Node(InputPin:Wildcard) -> (OutputPin:Wildcard)
  ]}
 EventGraph {nodes [
   Event BlueprintUpdateAnimation()
   > n1 = Sequence!(execute) -> (then_0, then_1, then_2:Exec, then_3:Exec, then_4:Exec, then_5:Exec, then_6:Exec, then_7:Exec)
   > n2 = Gravity Override State Function DEPRECATED(CharacterPartAnimInstance=Self-Reference(), GameTimeInSeconds=GetGameTimeInSeconds(), GravityOverrideParams=Get Gravity Override Params())
   > Set GravityOverride(GravityOverride=@n2)
   Set Hair Up(
    execute=@n1.then_1
    Hair Up:double=GetAlphaFromReferenceTransformDelta(SkeletalMeshComponent=Get Skel Mesh(), BoneName=spine_05, BaseName=head, InputRange=(X=0,Y=70), TwistAxis="0,0,1")
   ) -> (Output_Get:double)
   n3 = Get is Wearing Backbling()
   n4 = Branch(execute=@n1.then_4, Condition="Nearly Equal (Float)"(A="To Float (Boolean)"(InBool=@n3), B=1, ErrorTolerance=0.1))
   n5 = GetDeltaTransformFromRefPose(self=GetOwnerSkeletalMeshComponent(), BoneName=head, BaseName=neck_01)
   > Set head_fwdBwd_rot(head_fwdBwd_rot:double=BreakRotator(InRot=BreakTransform(InTransform=@n5).Rotation).Yaw) -> (Output_Get:double)
   Set head_fwdBwd_rot(execute=@n4.else, head_fwdBwd_rot=0)
   n6 = AND Boolean(A=Get bIsCrouching(), B=Get bIsMoving2D())
   n7 = Branch(execute=@n1.then_6, Condition="Nearly Equal (Float)"(A="To Float (Boolean)"(InBool=@n6), B=1, ErrorTolerance=0.1))
   n8 = GetOwnerSkeletalMeshComponent()
   n9 = BreakRotator(InRot=BreakTransform(InTransform=GetDeltaTransformFromRefPose(self=@n8, BoneName=head, BaseName=neck_01)).Rotation)
   > Set head_fwdBwd_rot(head_fwdBwd_rot="To Float (Boolean)"(InBool="float > float"(A=-41, B=@n9.Yaw)))
   Set head_fwdBwd_rot(execute=@n7.else, head_fwdBwd_rot=0)
   n10 = AND Boolean(A=Get bIsCrouching(), B=Get bIsTargeting())
   n11 = Get bIsMoving2D()
   Set ForwardCrouchAiming(execute=@n1.then_5, ForwardCrouchAiming:double="To Float (Boolean)"(InBool=AND Boolean(A=@n10, B=@n11))) -> (Output_Get:double)
   n12 = CheckForBackbling()
   Set is Wearing Backbling(
    execute=@n1.then_2
    is Wearing Backbling:bool=AND Boolean!(A=@n12, B=NOT Boolean(A=@n12.bIsBackblingACape), C:bool=NOT Boolean(A=@n12.bIsBackblingLong)) -> (ReturnValue)
   ) -> (Output_Get:bool)
   > Set is Wearing Cape(is Wearing Cape:bool=OR Boolean(A=@n12.bIsBackblingACape, B=@n12.bIsBackblingLong)) -> (Output_Get:bool)
   n13 = Get bIsSkydiveDiveMode()
   n14 = Get bIsSkydiveFloating()
   Set is Skydiving(execute=@n1.then_3, is Skydiving:double="To Float (Boolean)"(InBool=OR Boolean(A=@n13, B=@n14))) -> (Output_Get:double)
   > Set is Parachuting Alpha(is Parachuting Alpha:double="To Float (Boolean)"(InBool=Get bIsParachuteOpened())) -> (Output_Get:double)
   Set Pickaxe swing(execute=@n1.then_7, Pickaxe swing:double=SelectFloat(A=0.5, B=1, bPickA=Get bIsPlayingMeleeAnim())) -> (Output_Get:double)
   Event BlueprintInitializeAnimation()
   > Set Skel Mesh(Skel Mesh=GetOwnerSkeletalMeshComponent())
  ]}
]
variables [
 ForwardCrouchAiming {type=number}
 Gravity Override Params {type=GravityOverrideParamsStruct}
 GravityOverride {type=Vector}
 Hair Up {type=number}
 head_fwdBwd_rot {type=number}
 is Parachuting Alpha {type=number}
 is Skydiving {type=number}
 is Wearing Backbling {type=boolean}
 is Wearing Cape {type=boolean}
 Pickaxe swing {type=number}
 Skel Mesh {type=SkeletalMeshComponent}
]
